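# Niagara system template from Niagara Fluids for a 2D FLIP liquid, named as a pool setup.
# A single emitter, PIC_FLIP_Fluid, runs as a GPU compute simulation and draws through a sprite renderer.
# The emitter name indicates a hybrid PIC/FLIP particle-and-grid solver.
# The summary lists no user variables.
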NiagaraSystem /NiagaraFluids/Templates/Liquid/2D/Systems/Grid2D_FLIP_Pool
dependencies=null
summary {
 systemName=Grid2D_FLIP_Pool
 emitters [{emitterName=PIC_FLIP_Fluid, simTarget=GPUComputeSim, rendererClasses [&NiagaraSpriteRendererProperties]}]
 userVariables []
}
